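# Physics asset for the Raven critter backpack cosmetic, simulating its mirrored left and right wings and feathers with capsule bodies.
# Each wing chains from backpack_root through wing_02, wing_05 and wing_06.
# Feathers A to C hang off wing_06 as two-body chains, while D and E are single bodies on wing_06.
# All constraints set twist to zero, with swing limits widest at the wing root and narrow toward the feathers.
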PhysicsAsset /BRCosmetics/Accessories/FORT_Backpacks/Backpack_M_MED_Critter_Raven/Meshes/M_MED_Critter_Raven_Pack_Physics
bodies {
 backpack_root [{shapeType=Capsule, length=2, radius=2}]
 featherA_01_l [{shapeType=Capsule, length=2, radius=2, rotation=<rot 90 0 0>}]
 featherA_01_r [{shapeType=Capsule, length=2, radius=2, rotation=<rot 90 0 0>}]
 featherA_03_l [{shapeType=Capsule, length=3.058, radius=0.942, shapeName=featherA_03_l_capsule, rotation=<rot 90 0 0>}]
 featherA_03_r [{shapeType=Capsule, length=3.058, radius=0.942, shapeName=featherA_03_r_capsule, rotation=<rot 90 0 0>}]
 featherB_01_l [{shapeType=Capsule, length=2, radius=2, rotation=<rot 90 0 0>}]
 featherB_01_r [{shapeType=Capsule, length=2, radius=2, rotation=<rot 90 0 0>}]
 featherB_03_l [{shapeType=Capsule, length=3.058, radius=0.942, shapeName=featherB_03_l_capsule, rotation=<rot 90 0 0>}]
 featherB_03_r [{shapeType=Capsule, length=3.058, radius=0.942, shapeName=featherB_03_r_capsule, rotation=<rot 90 0 0>}]
 featherC_01_l [{shapeType=Capsule, length=2, radius=2, rotation=<rot 90 0 0>}]
 featherC_01_r [{shapeType=Capsule, length=2, radius=2, rotation=<rot 90 0 0>}]
 featherC_03_l [{shapeType=Capsule, length=3.058, radius=0.942, shapeName=featherC_03_l_capsule, rotation=<rot 90 0 0>}]
 featherC_03_r [{shapeType=Capsule, length=3.058, radius=0.942, shapeName=featherC_03_r_capsule, rotation=<rot 90 0 0>}]
 featherD_01_l [{shapeType=Capsule, length=2, radius=2, shapeName=featherD_01_l_capsule, rotation=<rot 90 0 0>}]
 featherD_01_r [{shapeType=Capsule, length=2, radius=2, shapeName=featherD_01_r_capsule, rotation=<rot 90 0 0>}]
 featherE_03_l [{shapeType=Capsule, length=3.058, radius=0.942, shapeName=featherE_03_l_capsule, rotation=<rot 90 0 0>}]
 featherE_03_r [{shapeType=Capsule, length=3.058, radius=0.942, shapeName=featherE_03_r_capsule, rotation=<rot 90 0 0>}]
 wing_02_l [{shapeType=Capsule, length=5.4949, radius=5, rotation=<rot 90 0 0>}]
 wing_02_r [{shapeType=Capsule, length=5.4949, radius=5, rotation=<rot 90 0 0>}]
 wing_05_l [{shapeType=Capsule, length=5, radius=3, shapeName=wing_05_l_capsule}]
 wing_05_r [{shapeType=Capsule, length=5, radius=3, shapeName=wing_05_r_capsule}]
 wing_06_l [{shapeType=Capsule, length=1.8574, radius=4.9058}]
 wing_06_r [{shapeType=Capsule, length=1.8574, radius=4.9058, center=<-1.8749 0 0>}]
}
constraints [
 {bone1Name=wing_02_l, bone2Name=backpack_root, swing1LimitDegrees=20, swing2LimitDegrees=20, twistLimitDegrees=0}
 {bone1Name=wing_06_l, bone2Name=wing_05_l, swing1LimitDegrees=2, swing2LimitDegrees=5, twistLimitDegrees=0}
 {bone1Name=featherA_01_l, bone2Name=wing_06_l, swing1LimitDegrees=2, swing2LimitDegrees=10, twistLimitDegrees=0}
 {bone1Name=featherB_01_l, bone2Name=wing_06_l, swing1LimitDegrees=2, swing2LimitDegrees=10, twistLimitDegrees=0}
 {bone1Name=featherC_01_l, bone2Name=wing_06_l, swing1LimitDegrees=2, swing2LimitDegrees=10, twistLimitDegrees=0}
 {bone1Name=wing_02_r, bone2Name=backpack_root, swing1LimitDegrees=20, swing2LimitDegrees=20, twistLimitDegrees=0}
 {bone1Name=wing_06_r, bone2Name=wing_05_r, swing1LimitDegrees=2, swing2LimitDegrees=5, twistLimitDegrees=0}
 {bone1Name=featherA_01_r, bone2Name=wing_06_r, swing1LimitDegrees=2, swing2LimitDegrees=10, twistLimitDegrees=0}
 {bone1Name=featherB_01_r, bone2Name=wing_06_r, swing1LimitDegrees=2, swing2LimitDegrees=10, twistLimitDegrees=0}
 {bone1Name=featherC_01_r, bone2Name=wing_06_r, swing1LimitDegrees=2, swing2LimitDegrees=10, twistLimitDegrees=0}
 {bone1Name=featherD_01_r, bone2Name=wing_06_r, swing1LimitDegrees=2, swing2LimitDegrees=10, twistLimitDegrees=0}
 {bone1Name=featherD_01_l, bone2Name=wing_06_l, swing1LimitDegrees=2, swing2LimitDegrees=10, twistLimitDegrees=0}
 {bone1Name=featherA_03_l, bone2Name=featherA_01_l, swing1LimitDegrees=2, swing2LimitDegrees=10, twistLimitDegrees=0}
 {bone1Name=featherB_03_l, bone2Name=featherB_01_l, swing1LimitDegrees=2, swing2LimitDegrees=10, twistLimitDegrees=0}
 {bone1Name=featherC_03_l, bone2Name=featherC_01_l, swing1LimitDegrees=2, swing2LimitDegrees=10, twistLimitDegrees=0}
 {bone1Name=featherE_03_l, bone2Name=wing_06_l, swing1LimitDegrees=2, swing2LimitDegrees=10, twistLimitDegrees=0}
 {bone1Name=featherA_03_r, bone2Name=featherA_01_r, swing1LimitDegrees=2, swing2LimitDegrees=10, twistLimitDegrees=0}
 {bone1Name=featherB_03_r, bone2Name=featherB_01_r, swing1LimitDegrees=2, swing2LimitDegrees=10, twistLimitDegrees=0}
 {bone1Name=featherC_03_r, bone2Name=featherC_01_r, swing1LimitDegrees=2, swing2LimitDegrees=10, twistLimitDegrees=0}
 {bone1Name=featherE_03_r, bone2Name=wing_06_r, swing1LimitDegrees=2, swing2LimitDegrees=10, twistLimitDegrees=0}
 {bone1Name=wing_05_r, bone2Name=wing_02_r, swing1LimitDegrees=10, swing2LimitDegrees=12, twistLimitDegrees=0}
 {bone1Name=wing_05_l, bone2Name=wing_02_l, swing1LimitDegrees=10, swing2LimitDegrees=12, twistLimitDegrees=0}
]
